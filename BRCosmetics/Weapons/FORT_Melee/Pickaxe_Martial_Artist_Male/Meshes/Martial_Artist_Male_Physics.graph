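# Physics asset for the male Martial Artist pickaxe, a two-handed melee cosmetic weapon with a dangling bottle.
# The handle body uses two capsules, a long one along the shaft and a shorter, wider, angled one offset near the upper end.
# Two dyn_bottle bodies, one capsule each, chain from the handle through limited swing and tighter twist constraints, letting the bottle swing as secondary motion.
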PhysicsAsset /BRCosmetics/Weapons/FORT_Melee/Pickaxe_Martial_Artist_Male/Meshes/Martial_Artist_Male_Physics
bodies {
 Melee_TwoHanded_Handle [
  {shapeType=Capsule, length=110.9474, radius=3.4085, center=<0 -0.4237 48.9517>, rotation=<rot 0 -90 0>}
  {shapeType=Capsule, length=15, radius=9.275, center=<0 -21.8894 91.7049>, rotation=<rot 0 0 -32.9705>}
 ]
 dyn_bottle_01 [{shapeType=Capsule, length=8, radius=2.9123, center=<6.4733 0 0>, rotation=<rot 90 0 0>}]
 dyn_bottle_02 [{shapeType=Capsule, length=15.5119, radius=6.0179, center=<13.7394 0 -0.0001>, rotation=<rot 90 0 0>}]
}
constraints [
 {bone1Name=dyn_bottle_02, bone2Name=dyn_bottle_01, swing1LimitDegrees=45, swing2LimitDegrees=45, twistLimitDegrees=10}
 {bone1Name=dyn_bottle_01, bone2Name=Melee_TwoHanded_Handle, swing1LimitDegrees=45, swing2LimitDegrees=45, twistLimitDegrees=10}
]
